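; Gameplay ability for a battle pass quest that credits the player for jumping while swimming.
; On event-triggered activation it binds to the owning pawn's jump event and waits for the swimming tag to be removed, which ends the ability.
; A jump sends a custom stat event with the quest source tag to the player's quest manager, then ends the ability.
; Ending the ability unbinds the jump event.
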(event Ability|EventActivateAbilityFromEvent
  (params (EventData :type Gameplay%20Event%20Data%20Structure%20%28by%20ref%29))
  ;[1 1 1] Activate ability - listen for when jumps while swimming
  (|CallParentFunction EventData)
  (Pawn|BindEventtoOnPawnJumped (Variables|GAVars|Data|GetOwningPlayerPawn) (EventDispatchers|CreateEvent :!event Custom|JumpTriggered))
  (Ability|Tasks|WaitGameplayTagRemove (Variables|Default|GetSwimming) 0 true
    (:Removed
      (Ability|EndAbility))))

(event Custom|JumpTriggered
  ;[1 1 1] Activate ability - listen for when jumps while swimming
  (Quest|SendComplexCustomStatEvent (Variables|GAVars|Data|GetOwningPlayerQuestManager) 0 (Variables|Default|GetQuestSourceTag) "(GameplayTags=)" 1)
  (Ability|EndAbility))

(event Ability|EventOnEndAbility
  (params (bWasCancelled :type Boolean))
  (Pawn|UnbindEventfromOnPawnJumped (Variables|GAVars|Data|GetOwningPlayerPawn) (EventDispatchers|CreateEvent :!event Custom|JumpTriggered)))
